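# Physics asset for the F_MED_Mainframe female medium body that defines physics for five dangling cloth bone chains (A to E).
# Thin box bodies form each chain, linked bone to bone by constraints with tight swing and twist limits.
# Chains C and D attach to the right thigh with stiffer limits, while A, B and E have no root constraint.
# Capsules cover only pelvis, right thigh and left thigh twist.
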PhysicsAsset /BRCosmetics/Characters/Player/Female/Medium/Bodies/F_MED_Mainframe/Meshes/F_MED_Mainframe_Physics1
bodies {
 dyn_clothA_01 [{shapeType=Box, extentX=7.5, extentY=0.6426, extentZ=7.5}]
 dyn_clothA_02 [{shapeType=Box, extentX=6.5566, extentY=0.5, extentZ=5.51, center=<0 0 -2.4516>}]
 dyn_clothA_03 [{shapeType=Box, extentX=6.5566, extentY=0.5, extentZ=5.51, center=<0 0 -2.4516>}]
 dyn_clothB_01 [{shapeType=Box, extentX=7.5, extentY=0.5, extentZ=7.5, center=<0 0.0627 -0.8335>, rotation=<rot 0.3671 4.8826 -4.2898>}]
 dyn_clothB_02 [{shapeType=Box, extentX=8.8648, extentY=0.5, extentZ=5.4413, center=<0 0.0684 -2.422>, rotation=<rot 0.2651 4.8893 -3.0966>}]
 dyn_clothB_03 [{shapeType=Box, extentX=8.8648, extentY=0.5, extentZ=5.4413, center=<0 0.0819 -2.7942>, rotation=<rot 0.1426 4.8944 -1.6645>}]
 dyn_clothB_04 [{shapeType=Box, extentX=8.8648, extentY=0.5, extentZ=5.4413, center=<0 -0.0435 -2.7952>, rotation=<rot 0.0264 4.8964 -0.3088>}]
 dyn_clothC_01 [{shapeType=Box, extentX=7.5, extentY=0.6426, extentZ=7.5}]
 dyn_clothC_02 [{shapeType=Box, extentX=7.5, extentY=1.0158, extentZ=4.9851, center=<0 0 -1.3711>}]
 dyn_clothC_03 [{shapeType=Box, extentX=7.5, extentY=1.0158, extentZ=4.9851, center=<0 0 -1.3711>}]
 dyn_clothD_01 [{shapeType=Box, extentX=7.5, extentY=0.6426, extentZ=7.5}]
 dyn_clothD_02 [{shapeType=Box, extentX=7.5, extentY=1.0158, extentZ=5.052, center=<0 0 -1.9721>}]
 dyn_clothD_03 [{shapeType=Box, extentX=7.5, extentY=1.0158, extentZ=5.052, center=<0 0 -1.9721>}]
 dyn_clothE_01 [{shapeType=Box, extentX=4.8826, extentY=0.5, extentZ=7.5, center=<0 0.2693 -3.428>, rotation=<rot 0 0 -4.4919>}]
 dyn_clothE_02 [{shapeType=Box, extentX=5.7536, extentY=0.5, extentZ=7.5, center=<0 0.1347 -3.3039>, rotation=<rot 0 0 -2.3339>}]
 dyn_clothE_03 [{shapeType=Box, extentX=5.7536, extentY=0.5, extentZ=7.5, center=<0 0 -3.3066>}]
 pelvis [{shapeType=Capsule, length=9.1121, radius=9.8761, center=<-3.4389 -2.0119 0.1142>, rotation=<rot 2.6818 -98.8704 1.3317>}]
 thigh_r [{shapeType=Capsule, length=17.5135, radius=7.2137, center=<12.8815 0.7033 0.2492>, rotation=<rot -0.7996 -98.1403 -96.231>}]
 thigh_twist_01_l [{shapeType=Capsule, length=19.5566, radius=7.5517, center=<18.7815 -0.1131 0.0207>, rotation=<rot -88.3445 -13.4306 8.7566>}]
}
constraints [
 {bone1Name=dyn_clothC_02, bone2Name=dyn_clothC_01, swing1LimitDegrees=5, swing2LimitDegrees=5, twistLimitDegrees=15}
 {bone1Name=dyn_clothC_03, bone2Name=dyn_clothC_02, swing1LimitDegrees=5, swing2LimitDegrees=5, twistLimitDegrees=15}
 {bone1Name=dyn_clothD_02, bone2Name=dyn_clothD_01, swing1LimitDegrees=5, swing2LimitDegrees=5, twistLimitDegrees=15}
 {bone1Name=dyn_clothD_03, bone2Name=dyn_clothD_02, swing1LimitDegrees=5, swing2LimitDegrees=5, twistLimitDegrees=15}
 {bone1Name=dyn_clothA_02, bone2Name=dyn_clothA_01, swing1LimitDegrees=5, swing2LimitDegrees=5, twistLimitDegrees=15}
 {bone1Name=dyn_clothA_03, bone2Name=dyn_clothA_02, swing1LimitDegrees=5, swing2LimitDegrees=5, twistLimitDegrees=15}
 {bone1Name=dyn_clothB_02, bone2Name=dyn_clothB_01, swing1LimitDegrees=5, swing2LimitDegrees=5, twistLimitDegrees=15}
 {bone1Name=dyn_clothB_03, bone2Name=dyn_clothB_02, swing1LimitDegrees=5, swing2LimitDegrees=5, twistLimitDegrees=15}
 {bone1Name=dyn_clothB_04, bone2Name=dyn_clothB_03, swing1LimitDegrees=5, swing2LimitDegrees=5, twistLimitDegrees=15}
 {bone1Name=dyn_clothC_01, bone2Name=thigh_r, swing1LimitDegrees=1, swing2LimitDegrees=1, twistLimitDegrees=5}
 {bone1Name=dyn_clothD_01, bone2Name=thigh_r, swing1LimitDegrees=1, swing2LimitDegrees=1, twistLimitDegrees=5}
 {bone1Name=dyn_clothE_02, bone2Name=dyn_clothE_01, swing1LimitDegrees=5, swing2LimitDegrees=5, twistLimitDegrees=15}
 {bone1Name=dyn_clothE_03, bone2Name=dyn_clothE_02, swing1LimitDegrees=5, swing2LimitDegrees=5, twistLimitDegrees=15}
]
