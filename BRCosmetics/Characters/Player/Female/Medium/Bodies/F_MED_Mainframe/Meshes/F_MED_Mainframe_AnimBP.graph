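# Post-process style AnimBlueprint for the female Mainframe body cosmetic, copying the parent mesh pose and adding corrective and cloth motion.
# Each update converts thigh and upper arm rotation deltas from the reference pose, relative to pelvis and clavicles, into alphas.
# These alphas drive mostly additive bone offsets on thigh, shoulder and dyn_cloth helper bones, after two left-thigh constraints.
# Two pelvis-space RigidBody nodes with separate physics assets then add secondary simulation.
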AnimBlueprint /BRCosmetics/Characters/Player/Female/Medium/Bodies/F_MED_Mainframe/Meshes/F_MED_Mainframe_AnimBP
target_skeleton=&/F_MED_Mainframe_Skeleton
graphs [
 AnimGraph {nodes [
   n1 = "Constraint - thigh_l_rot (PP)"!(
    ComponentPose=Local To Component(LocalPose=Copy Pose From Mesh() {bUseAttachedParent=true})
    Alpha=1
    ConstraintWeights_0=0.5
    ConstraintWeights_1:float=0.5
   ) -> (Pose) {
    alphaBoolBlend {blendOption=Linear}
    boneToModify {boneName=thigh_l_rot}
    constraintSetup [{targetBone {boneName=thigh_twist_01_l}}, {targetBone {boneName=pelvis}}]
    constraintWeights [1,1]
   }
   n2 = "Constraint - thigh_l_f (PP)"!(ComponentPose=@n1, Alpha=1, ConstraintWeights_0=0.4, ConstraintWeights_1:float=0.6) -> (Pose) {
    alphaBoolBlend {blendOption=Linear}
    boneToModify {boneName=thigh_l_f}
    constraintSetup [{targetBone {boneName=thigh_twist_01_l}}, {targetBone {boneName=pelvis}}]
    constraintWeights [1,1]
   }
   n3 = "Transform (Modify) Bone - Bone: thigh_volume_bk_l"(ComponentPose=@n2, Alpha=Get thigh_rot_l(), Translation="12,-5,0", Rotation="0,40,0", Scale="1,1,1") {
    rotationMode=BMM_Additive
    rotationSpace=BCS_BoneSpace
    scaleMode=BMM_Additive
    scaleSpace=BCS_BoneSpace
    translationMode=BMM_Additive
    translationSpace=BCS_BoneSpace
    alphaBoolBlend {blendOption=Linear}
    boneToModify {boneName=thigh_volume_bk_l}
    rotation {yaw=40}
    translation=<12 -5>
   }
   n4 = Get leg_rot_l() -> (leg_rot_l:double)
   n5 = "Transform (Modify) Bone - Bone: dyn_clothA_01"(ComponentPose=@n3, Alpha=@n4, Translation="0,-2,-2", Rotation="0,0,15", Scale="1,1,1") {
    rotationMode=BMM_Additive
    rotationSpace=BCS_BoneSpace
    scaleMode=BMM_Additive
    scaleSpace=BCS_BoneSpace
    translationMode=BMM_Additive
    translationSpace=BCS_BoneSpace
    alphaBoolBlend {blendOption=Linear}
    boneToModify {boneName=dyn_clothA_01}
    rotation {roll=15}
    translation {y=-2, z=-2}
   }
   n6 = "Transform (Modify) Bone - Bone: dyn_clothB_01"(ComponentPose=@n5, Alpha=@n4, Translation="1,0,-1", Rotation="-5,0,-15", Scale="1,1,1") {
    rotationMode=BMM_Additive
    rotationSpace=BCS_BoneSpace
    scaleMode=BMM_Additive
    scaleSpace=BCS_BoneSpace
    translationMode=BMM_Additive
    translationSpace=BCS_BoneSpace
    alphaBoolBlend {blendOption=Linear}
    boneToModify {boneName=dyn_clothB_01}
    rotation {pitch=-5, roll=-15}
    translation {x=1, z=-1}
   }
   n7 = "Transform (Modify) Bone - Bone: dyn_clothE_01"(ComponentPose=@n6, Alpha=@n4, Translation="0,0,0", Rotation="-15,10,-30", Scale="1,1,1") {
    rotationMode=BMM_Additive
    rotationSpace=BCS_BoneSpace
    scaleMode=BMM_Additive
    scaleSpace=BCS_BoneSpace
    translationMode=BMM_Additive
    translationSpace=BCS_BoneSpace
    alphaBoolBlend {blendOption=Linear}
    boneToModify {boneName=dyn_clothE_01}
    rotation=<rot -15 10 -30>
   }
   n8 = "Transform (Modify) Bone - Bone: thigh_l_rot"(ComponentPose=@n7, Alpha=@n4, Translation="-5,8,-2", Rotation="0,-55,0", Scale="1,1,1") {
    rotationMode=BMM_Additive
    rotationSpace=BCS_BoneSpace
    scaleMode=BMM_Additive
    scaleSpace=BCS_BoneSpace
    translationMode=BMM_Additive
    translationSpace=BCS_BoneSpace
    alphaBoolBlend {blendOption=Linear}
    boneToModify {boneName=thigh_l_rot}
    rotation {yaw=-55}
    translation=<-5 8 -2>
   }
   n9 = "Transform (Modify) Bone - Bone: thigh_l_f"(ComponentPose=@n8, Alpha=Get leg_rot_twish_l_01() -> (leg_rot_twish_l_01:double), Translation="10,0,0", Rotation="0,0,0", Scale="1,1,1") {
    rotationMode=BMM_Additive
    rotationSpace=BCS_BoneSpace
    scaleMode=BMM_Additive
    scaleSpace=BCS_BoneSpace
    translationMode=BMM_Additive
    translationSpace=BCS_BoneSpace
    alphaBoolBlend {blendOption=Linear}
    boneToModify {boneName=thigh_l_f}
    translation {x=10}
   }
   n10 = "Transform (Modify) Bone - Bone: thigh_l_f"(ComponentPose=@n9, Alpha=Get leg_rot_twish_l() -> (leg_rot_twish_l:double), Translation="0,15,0", Rotation="0,0,0", Scale="1,1,1") {
    rotationMode=BMM_Additive
    rotationSpace=BCS_BoneSpace
    scaleMode=BMM_Additive
    scaleSpace=BCS_BoneSpace
    translationMode=BMM_Additive
    translationSpace=BCS_BoneSpace
    alphaBoolBlend {blendOption=Linear}
    boneToModify {boneName=thigh_l_f}
    translation {y=15}
   }
   n11 = "Transform (Modify) Bone - Bone: thigh_volume_bk_r"(ComponentPose=@n10, Alpha=Get thigh_rot_r(), Translation="3,5,0", Rotation="0,0,0", Scale="1,1,1") {
    rotationSpace=BCS_BoneSpace
    scaleMode=BMM_Additive
    scaleSpace=BCS_BoneSpace
    translationMode=BMM_Additive
    translationSpace=BCS_BoneSpace
    alphaBoolBlend {blendOption=Linear}
    boneToModify {boneName=thigh_volume_bk_r}
    translation=<3 5>
   }
   n12 = Get leg_rot_r() -> (leg_rot_r:double)
   n13 = "Transform (Modify) Bone - Bone: dyn_clothC_01"(ComponentPose=@n11, Alpha=@n12, Translation="0,-2,0", Rotation="0,0,-50", Scale="1,1,1") {
    rotationMode=BMM_Additive
    rotationSpace=BCS_BoneSpace
    scaleMode=BMM_Additive
    scaleSpace=BCS_BoneSpace
    translationMode=BMM_Additive
    translationSpace=BCS_BoneSpace
    alphaBoolBlend {blendOption=Linear}
    boneToModify {boneName=dyn_clothC_01}
    rotation {roll=-50}
    translation {y=-2}
   }
   n14 = "Transform (Modify) Bone - Bone: dyn_clothD_01"(ComponentPose=@n13, Alpha=@n12, Translation="0,0,0", Rotation="0,0,0", Scale="1,1,1") {
    rotationMode=BMM_Additive
    rotationSpace=BCS_BoneSpace
    scaleMode=BMM_Additive
    scaleSpace=BCS_BoneSpace
    translationMode=BMM_Additive
    translationSpace=BCS_BoneSpace
    alphaBoolBlend {blendOption=Linear}
    boneToModify {boneName=dyn_clothD_01}
   }
   n15 = "Transform (Modify) Bone - Bone: flont_uppelalm_l"(ComponentPose=@n14, Alpha=Get Shoulder_rot_f_l() -> (Shoulder_rot_f_l:double), Translation="5,5,-5", Rotation="0,0,0", Scale="1,1,1") {
    rotationMode=BMM_Additive
    rotationSpace=BCS_BoneSpace
    scaleMode=BMM_Additive
    scaleSpace=BCS_BoneSpace
    translationMode=BMM_Additive
    translationSpace=BCS_BoneSpace
    alphaBoolBlend {blendOption=Linear}
    boneToModify {boneName=flont_uppelalm_l}
    translation=<5 5 -5>
   }
   n16 = "Transform (Modify) Bone - Bone: front_upperarm_r"(ComponentPose=@n15, Alpha=Get Shoulder_rot_f_r() -> (Shoulder_rot_f_r:double), Translation="-5,-5,5", Rotation="0,0,0", Scale="1,1,1") {
    rotationMode=BMM_Additive
    rotationSpace=BCS_BoneSpace
    scaleMode=BMM_Additive
    scaleSpace=BCS_BoneSpace
    translationMode=BMM_Additive
    translationSpace=BCS_BoneSpace
    alphaBoolBlend {blendOption=Linear}
    boneToModify {boneName=front_upperarm_r}
    translation=<-5 -5 5>
   }
   n17 = "Transform (Modify) Bone - Bone: up_uppelalm_l"(ComponentPose=@n16, Alpha=Get Shoulder_rot_up_l() -> (Shoulder_rot_up_l:double), Translation="0,2,2", Rotation="0,0,0", Scale="1,1,1") {
    rotationMode=BMM_Additive
    rotationSpace=BCS_BoneSpace
    scaleMode=BMM_Additive
    scaleSpace=BCS_BoneSpace
    translationMode=BMM_Additive
    translationSpace=BCS_BoneSpace
    alphaBoolBlend {blendOption=Linear}
    boneToModify {boneName=up_uppelalm_l}
    translation {y=2, z=2}
   }
   n18 = "Transform (Modify) Bone - Bone: up_upperarm_r"(ComponentPose=@n17, Alpha=Get Shoulder_rot_up_r() -> (Shoulder_rot_up_r:double), Translation="0,-2,-2", Rotation="0,0,0", Scale="1,1,1") {
    rotationMode=BMM_Additive
    rotationSpace=BCS_BoneSpace
    scaleMode=BMM_Additive
    scaleSpace=BCS_BoneSpace
    translationMode=BMM_Additive
    translationSpace=BCS_BoneSpace
    alphaBoolBlend {blendOption=Linear}
    boneToModify {boneName=up_upperarm_r}
    translation {y=-2, z=-2}
   }
   n19 = RigidBody(ComponentPose=@n18, Alpha=0.5) {
    bClampLinearTranslationLimitToRefPose=true
    bOverrideWorldGravity=true
    overridePhysicsAsset=&/F_MED_Mainframe_Physics
    simulationSpace=BaseBoneSpace
    alphaBoolBlend {blendOption=Linear}
    baseBoneRef {boneName=pelvis}
    componentAppliedLinearAccClamp=<1000 1000 1000>
    componentLinearAccScale=<0.35 0.45 0.45>
    componentLinearVelScale=<0.3 0.3 0.3>
    overrideWorldGravity {z=-980}
   }
   n20 = RigidBody(ComponentPose=@n19) {
    bClampLinearTranslationLimitToRefPose=true
    bOverrideWorldGravity=true
    overridePhysicsAsset=&/F_MED_Mainframe_Physics1
    simulationSpace=BaseBoneSpace
    alphaBoolBlend {blendOption=Linear}
    baseBoneRef {boneName=pelvis}
    componentAppliedLinearAccClamp=<1000 1000 1000>
    componentLinearAccScale=<0.35 0.45 0.45>
    componentLinearVelScale=<0.3 0.3 0.3>
    overrideWorldGravity {z=-980}
   }
   Output Pose(Result=Component To Local(ComponentPose=@n20)) {layerGroup=None}
  ]}
 EventGraph {nodes [
   Event BlueprintInitializeAnimation()
   > Set Skel_Mesh(Skel_Mesh=GetOwnerSkeletalMeshComponent())
   Event BlueprintUpdateAnimation()
   n1 = Get Skel_Mesh()
   > Set thigh_rot_r(thigh_rot_r=GetAlphaFromReferenceTransformDelta(SkeletalMeshComponent=@n1, BoneName=thigh_r, BaseName=pelvis, InputRange=(X=0,Y=-90), TwistAxis="0,0,1"))
   > Set thigh_rot_l(thigh_rot_l=GetAlphaFromReferenceTransformDelta(SkeletalMeshComponent=@n1, BoneName=thigh_l, BaseName=pelvis, InputRange=(X=-15,Y=-110), TwistAxis="0,0,1"))
   n2 = Get Skel_Mesh()
   > Set Shoulder_rot_f_r(Shoulder_rot_f_r:double=GetAlphaFromReferenceTransformDelta(SkeletalMeshComponent=@n2, BoneName=upperarm_r, BaseName=clavicle_r, InputRange=(X=0,Y=90), TwistAxis="0,0,1")) -> (Output_Get:double)
   > Set Shoulder_rot_f_l(Shoulder_rot_f_l:double=GetAlphaFromReferenceTransformDelta(SkeletalMeshComponent=@n2, BoneName=upperarm_l, BaseName=clavicle_l, InputRange=(X=0,Y=90), TwistAxis="0,0,1")) -> (Output_Get:double)
   n3 = Get Skel_Mesh()
   > Set Shoulder_rot_up_r(Shoulder_rot_up_r:double=GetAlphaFromReferenceTransformDelta(
     SkeletalMeshComponent=@n3
     BoneName=upperarm_twist_01_r
     BaseName=clavicle_r
     DeltaComponentType=EulerY
     InputRange=(X=0,Y=110)
     TwistAxis="0,0,1"
    )) -> (Output_Get:double)
   > Set Shoulder_rot_up_l(Shoulder_rot_up_l:double=GetAlphaFromReferenceTransformDelta(SkeletalMeshComponent=@n3, BoneName=upperarm_l, BaseName=clavicle_l, DeltaComponentType=EulerY, InputRange=(X=0,Y=110), TwistAxis="0,0,1")) -> (Output_Get:double)
   n4 = Get Skel_Mesh()
   > Set leg_rot_r(leg_rot_r:double=GetAlphaFromReferenceTransformDelta(SkeletalMeshComponent=@n4, BoneName=thigh_r, BaseName=pelvis, InputRange=(X=0,Y=-90), TwistAxis="0,0,1")) -> (Output_Get:double)
   > Set leg_rot_l(leg_rot_l:double=GetAlphaFromReferenceTransformDelta(SkeletalMeshComponent=@n4, BoneName=thigh_l, BaseName=pelvis, InputRange=(X=0,Y=-150), TwistAxis="0,0,1")) -> (Output_Get:double)
   n5 = Get Skel_Mesh()
   > Set leg_rot_twish_l(leg_rot_twish_l:double=GetAlphaFromReferenceTransformDelta(SkeletalMeshComponent=@n5, BoneName=thigh_l, BaseName=pelvis, InputRange=(X=0,Y=0), OutputRange=(X=0,Y=0), TwistAxis="0,0,1")) -> (Output_Get:double)
   > Set leg_rot_twish_l_01(leg_rot_twish_l_01:double=GetAlphaFromReferenceTransformDelta(SkeletalMeshComponent=@n5, BoneName=thigh_l, BaseName=pelvis, InputRange=(X=-60,Y=-120), TwistAxis="0,0,1")) -> (Output_Get:double)
  ]}
]
variables [
 leg_rot_l {type=number}
 leg_rot_r {type=number}
 leg_rot_twish_l {type=number}
 leg_rot_twish_l_01 {type=number}
 Shoulder_rot_f_l {type=number}
 Shoulder_rot_f_r {type=number}
 Shoulder_rot_up_l {type=number}
 Shoulder_rot_up_r {type=number}
 Skel_Mesh {type=SkeletalMeshComponent}
 thigh_rot_l {type=number}
 thigh_rot_r {type=number}
]
